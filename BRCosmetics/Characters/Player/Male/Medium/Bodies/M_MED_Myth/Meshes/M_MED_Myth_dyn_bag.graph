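# Physics asset for the bag on the medium male Myth player character body, covering its dyn_bag bones plus pelvis and spine_01.
# Large pelvis and spine capsules stand in for the torso, while a tiny master capsule, a box base and a capsule flap form the bag.
# Constraints chain flap to base to master to spine_01, with angularly free motion at the spine attachment and narrow limits within the bag.
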PhysicsAsset /BRCosmetics/Characters/Player/Male/Medium/Bodies/M_MED_Myth/Meshes/M_MED_Myth_dyn_bag
bodies {
 dyn_bag_MSTR [{shapeType=Capsule, length=0.505, radius=0.505, rotation=<rot 0 -90 0>}]
 dyn_bag_base [{shapeType=Box, extentX=5.5, extentY=16, extentZ=3, center=<4.3345 0.0001 1.2925>, rotation=<rot 0.9691 0 0>}]
 dyn_bag_flap [{shapeType=Capsule, length=3.25, radius=1.5, center=<4.8842 0.0001 0>, rotation=<rot 0 0 91.5608>}]
 pelvis [{shapeType=Capsule, length=9.5, radius=10.75, center=<0 -1.7149 0>}]
 spine_01 [{shapeType=Capsule, length=7.25, radius=9.5, center=<0 -0.7738 0.0001>}]
}
constraints [
 {bone1Name=spine_01, bone2Name=pelvis, swing1LimitDegrees=45, swing2LimitDegrees=45, twistLimitDegrees=45}
 {
  bone1Name=dyn_bag_MSTR
  bone2Name=spine_01
  swing1LimitDegrees=45
  swing1Motion=Free
  swing2LimitDegrees=45
  swing2Motion=Free
  twistLimitDegrees=45
  twistMotion=Free
 }
 {bone1Name=dyn_bag_base, bone2Name=dyn_bag_MSTR, swing1LimitDegrees=5, swing2LimitDegrees=20, twistLimitDegrees=5}
 {bone1Name=dyn_bag_flap, bone2Name=dyn_bag_base, swing1LimitDegrees=5, swing2LimitDegrees=30, twistLimitDegrees=5}
]
