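# Physics asset for the dynamic band bones on the Myth male medium player body cosmetic, covering both left and right bands despite the "band_l" name.
# Each side has a master bone branching into two two-bone capsule chains, with large spine_04 and spine_05 capsules as torso bodies.
# Chain joints allow little twist, with tighter swing on the left than the right, while master-to-spine and spine-to-spine joints rotate freely.
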PhysicsAsset /BRCosmetics/Characters/Player/Male/Medium/Bodies/M_MED_Myth/Meshes/M_MED_Myth_dyn_band_l
bodies {
 dyn_band_l_MSTR [{shapeType=Capsule, length=1, radius=1}]
 dyn_band_l_a_01 [{shapeType=Capsule, length=2, radius=1, center=<2 0 0>, rotation=<rot 90 0 0>}]
 dyn_band_l_a_02 [{shapeType=Capsule, length=3, radius=1, center=<3 0 0>, rotation=<rot 90 0 0>}]
 dyn_band_l_b_01 [{shapeType=Capsule, length=2, radius=1, center=<2 0 0>, rotation=<rot 90 0 0>}]
 dyn_band_l_b_02 [{shapeType=Capsule, length=3, radius=1, center=<3 0 0>, rotation=<rot 90 0 0>}]
 dyn_band_r_MSTR [{shapeType=Capsule, length=1, radius=1}]
 dyn_band_r_a_01 [{shapeType=Capsule, length=2, radius=1, center=<2 0 0>, rotation=<rot 90 0 0>}]
 dyn_band_r_a_02 [{shapeType=Capsule, length=3, radius=1, center=<3 0 0>, rotation=<rot 90 0 0>}]
 dyn_band_r_b_01 [{shapeType=Capsule, length=2, radius=1, center=<2 0 0>, rotation=<rot 90 0 0>}]
 dyn_band_r_b_02 [{shapeType=Capsule, length=3, radius=1, center=<3 0 0>, rotation=<rot 90 0 0>}]
 spine_04 [{shapeType=Capsule, length=10.2077, radius=13.7928, center=<3.184 0.5483 0.063>, rotation=<rot -0.3689 -90 0.1167>}]
 spine_05 [{shapeType=Capsule, length=13.7622, radius=13.2574, center=<3.5983 -0.1145 -0.0026>, rotation=<rot -0.2951 -90 -0.3901>}]
}
constraints [
 {bone1Name=dyn_band_l_a_01, bone2Name=dyn_band_l_MSTR, swing1LimitDegrees=45, swing2LimitDegrees=15, twistLimitDegrees=5}
 {bone1Name=dyn_band_l_a_02, bone2Name=dyn_band_l_a_01, swing1LimitDegrees=45, swing2LimitDegrees=15, twistLimitDegrees=5}
 {bone1Name=dyn_band_l_b_01, bone2Name=dyn_band_l_MSTR, swing1LimitDegrees=45, swing2LimitDegrees=15, twistLimitDegrees=5}
 {bone1Name=dyn_band_l_b_02, bone2Name=dyn_band_l_b_01, swing1LimitDegrees=45, swing2LimitDegrees=15, twistLimitDegrees=5}
 {bone1Name=dyn_band_r_a_01, bone2Name=dyn_band_r_MSTR, swing1LimitDegrees=60, swing2LimitDegrees=60, twistLimitDegrees=5}
 {bone1Name=dyn_band_r_a_02, bone2Name=dyn_band_r_a_01, swing1LimitDegrees=45, swing2LimitDegrees=45, twistLimitDegrees=5}
 {bone1Name=dyn_band_r_b_01, bone2Name=dyn_band_r_MSTR, swing1LimitDegrees=60, swing2LimitDegrees=60, twistLimitDegrees=5}
 {bone1Name=dyn_band_r_b_02, bone2Name=dyn_band_r_b_01, swing1LimitDegrees=45, swing2LimitDegrees=45, twistLimitDegrees=5}
 {
  bone1Name=dyn_band_l_MSTR
  bone2Name=spine_05
  swing1LimitDegrees=45
  swing1Motion=Free
  swing2LimitDegrees=45
  swing2Motion=Free
  twistLimitDegrees=45
  twistMotion=Free
 }
 {
  bone1Name=dyn_band_r_MSTR
  bone2Name=spine_05
  swing1LimitDegrees=45
  swing1Motion=Free
  swing2LimitDegrees=45
  swing2Motion=Free
  twistLimitDegrees=45
  twistMotion=Free
 }
 {
  bone1Name=spine_05
  bone2Name=spine_04
  swing1LimitDegrees=45
  swing1Motion=Free
  swing2LimitDegrees=45
  swing2Motion=Free
  twistLimitDegrees=45
  twistMotion=Free
 }
]
