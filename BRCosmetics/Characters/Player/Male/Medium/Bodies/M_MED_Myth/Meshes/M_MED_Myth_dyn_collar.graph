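# Physics asset for the dynamic collar on the M_MED_Myth male player body, with box bodies for a back piece and three pieces per side.
# Each piece hangs from a small _MATR helper body on spine_05 through a hinge-like constraint with tight twist and one wide swing axis.
# Head, neck and left clavicle capsules supply body shapes around the collar; there is no right clavicle body.
# Helper, neck, head and clavicle constraints leave rotation free.
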PhysicsAsset /BRCosmetics/Characters/Player/Male/Medium/Bodies/M_MED_Myth/Meshes/M_MED_Myth_dyn_collar
bodies {
 clavicle_l [{shapeType=Capsule, length=4.8716, radius=4.5219, center=<15.1188 -0.0001 0>, rotation=<rot 0 -5.0006 89.9999>}]
 dyn_collar_bk [{shapeType=Box, extentX=7.5, extentY=12.5, extentZ=5.5, center=<4.0986 0 -0.5891>, rotation=<rot 10 0 0>}]
 dyn_collar_bk_MATR [{shapeType=Capsule, length=0.505, radius=0.505, rotation=<rot 0 -90 0>}]
 dyn_collar_l_01 [{shapeType=Box, extentX=7, extentY=9.5, extentZ=5, center=<3.8739 0.0363 0.0002>, rotation=<rot -5 -5 -5>}]
 dyn_collar_l_01_MATR [{shapeType=Capsule, length=0.505, radius=0.505, rotation=<rot 0 -90 0>}]
 dyn_collar_l_02 [{shapeType=Box, extentX=6.5, extentY=12.5, extentZ=4.5, center=<3.0769 0.0626 0.7047>, rotation=<rot 5.0188 0.4369 -9.9434>}]
 dyn_collar_l_02_MATR [{shapeType=Capsule, length=0.505, radius=0.505, rotation=<rot 0 -90 0>}]
 dyn_collar_l_03 [{shapeType=Box, extentX=5, extentY=12, extentZ=5, center=<1.891 1.2266 0.0004>, rotation=<rot 5 -24.9983 -0.0001>}]
 dyn_collar_l_03_MATR [{shapeType=Capsule, length=0.505, radius=0.505, rotation=<rot 0 -90 0>}]
 dyn_collar_r_01 [{shapeType=Box, extentX=7, extentY=9.5, extentZ=5, center=<3.8739 0.0363 0.0002>, rotation=<rot -5 5.0002 5>}]
 dyn_collar_r_01_MATR [{shapeType=Capsule, length=0.505, radius=0.505, rotation=<rot 0 -90 0>}]
 dyn_collar_r_02 [{shapeType=Box, extentX=6.5, extentY=9.5, extentZ=4.5, center=<3.0769 0.0626 0.7047>, rotation=<rot 5.0188 -0.4369 4.943>}]
 dyn_collar_r_02_MATR [{shapeType=Capsule, length=0.505, radius=0.505, rotation=<rot 0 -90 0>}]
 dyn_collar_r_03 [{shapeType=Box, extentX=5, extentY=6.5, extentZ=5, center=<1.891 -1.2266 0.0004>, rotation=<rot 5 24.9984 0>}]
 dyn_collar_r_03_MATR [{shapeType=Capsule, length=0.505, radius=0.505, rotation=<rot 0 -90 0>}]
 head [
  {shapeType=Capsule, length=4.75, radius=8, center=<7.6922 1.6766 0>, rotation=<rot 90 0 0>}
  {shapeType=Capsule, length=6.75, radius=6, center=<3.5326 6.7325 -0.0001>, rotation=<rot 90 0 0>}
 ]
 neck_01 [{shapeType=Capsule, length=0.5, radius=4.75, rotation=<rot 90 0 0>}]
 neck_02 [{shapeType=Capsule, length=0.5, radius=4.75, rotation=<rot 90 0 0>}]
 spine_05 [{shapeType=Capsule, length=0.5, radius=0.5}]
}
constraints [
 {
  bone1Name=dyn_collar_bk_MATR
  bone2Name=spine_05
  swing1LimitDegrees=45
  swing1Motion=Free
  swing2LimitDegrees=45
  swing2Motion=Free
  twistLimitDegrees=45
  twistMotion=Free
 }
 {bone1Name=dyn_collar_bk, bone2Name=dyn_collar_bk_MATR, swing1LimitDegrees=5, swing2LimitDegrees=75, twistLimitDegrees=2}
 {
  bone1Name=dyn_collar_r_01_MATR
  bone2Name=spine_05
  swing1LimitDegrees=45
  swing1Motion=Free
  swing2LimitDegrees=45
  swing2Motion=Free
  twistLimitDegrees=45
  twistMotion=Free
 }
 {bone1Name=dyn_collar_r_01, bone2Name=dyn_collar_r_01_MATR, swing1LimitDegrees=5, swing2LimitDegrees=75, twistLimitDegrees=2}
 {
  bone1Name=dyn_collar_r_02_MATR
  bone2Name=spine_05
  swing1LimitDegrees=45
  swing1Motion=Free
  swing2LimitDegrees=45
  swing2Motion=Free
  twistLimitDegrees=45
  twistMotion=Free
 }
 {bone1Name=dyn_collar_r_02, bone2Name=dyn_collar_r_02_MATR, swing1LimitDegrees=5, swing2LimitDegrees=45, twistLimitDegrees=2}
 {
  bone1Name=dyn_collar_r_03_MATR
  bone2Name=spine_05
  swing1LimitDegrees=45
  swing1Motion=Free
  swing2LimitDegrees=45
  swing2Motion=Free
  twistLimitDegrees=45
  twistMotion=Free
 }
 {bone1Name=dyn_collar_r_03, bone2Name=dyn_collar_r_03_MATR, swing1LimitDegrees=5, swing2LimitDegrees=45, twistLimitDegrees=2}
 {
  bone1Name=dyn_collar_l_01_MATR
  bone2Name=spine_05
  swing1LimitDegrees=45
  swing1Motion=Free
  swing2LimitDegrees=45
  swing2Motion=Free
  twistLimitDegrees=45
  twistMotion=Free
 }
 {bone1Name=dyn_collar_l_01, bone2Name=dyn_collar_l_01_MATR, swing1LimitDegrees=5, swing2LimitDegrees=75, twistLimitDegrees=2}
 {
  bone1Name=dyn_collar_l_02_MATR
  bone2Name=spine_05
  swing1LimitDegrees=45
  swing1Motion=Free
  swing2LimitDegrees=45
  swing2Motion=Free
  twistLimitDegrees=45
  twistMotion=Free
 }
 {bone1Name=dyn_collar_l_02, bone2Name=dyn_collar_l_02_MATR, swing1LimitDegrees=5, swing2LimitDegrees=45, twistLimitDegrees=2}
 {
  bone1Name=dyn_collar_l_03_MATR
  bone2Name=spine_05
  swing1LimitDegrees=45
  swing1Motion=Free
  swing2LimitDegrees=45
  swing2Motion=Free
  twistLimitDegrees=45
  twistMotion=Free
 }
 {bone1Name=dyn_collar_l_03, bone2Name=dyn_collar_l_03_MATR, swing1LimitDegrees=5, swing2LimitDegrees=45, twistLimitDegrees=2}
 {
  bone1Name=neck_01
  bone2Name=spine_05
  swing1LimitDegrees=45
  swing1Motion=Free
  swing2LimitDegrees=45
  swing2Motion=Free
  twistLimitDegrees=45
  twistMotion=Free
 }
 {
  bone1Name=neck_02
  bone2Name=neck_01
  swing1LimitDegrees=45
  swing1Motion=Free
  swing2LimitDegrees=45
  swing2Motion=Free
  twistLimitDegrees=45
  twistMotion=Free
 }
 {
  bone1Name=head
  bone2Name=neck_02
  swing1LimitDegrees=45
  swing1Motion=Free
  swing2LimitDegrees=45
  swing2Motion=Free
  twistLimitDegrees=45
  twistMotion=Free
 }
 {
  bone1Name=clavicle_l
  bone2Name=spine_05
  swing1LimitDegrees=45
  swing1Motion=Free
  swing2LimitDegrees=45
  swing2Motion=Free
  twistLimitDegrees=45
  twistMotion=Free
 }
]
