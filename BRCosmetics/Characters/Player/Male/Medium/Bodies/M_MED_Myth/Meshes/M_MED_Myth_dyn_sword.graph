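# Physics asset for the dangling sword on the M_MED_Myth male medium body, used for secondary motion of the dyn_sword bone.
# A box body covers the sword, constrained with tight, asymmetric swing and twist limits to a tiny dyn_sword_MATR anchor capsule.
# Capsules on the right thigh and calf give leg volumes beside the sword.
# The anchor and calf each connect to the thigh through angularly free constraints.
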PhysicsAsset /BRCosmetics/Characters/Player/Male/Medium/Bodies/M_MED_Myth/Meshes/M_MED_Myth_dyn_sword
bodies {
 calf_r [{shapeType=Capsule, length=29.4334, radius=5.7543, center=<18.9158 0.0001 0>, rotation=<rot 90 0 0>}]
 dyn_sword [{shapeType=Box, extentX=65.5, extentY=5.5, extentZ=3, center=<8.4801 0 0>}]
 dyn_sword_MATR [{shapeType=Capsule, length=0.505, radius=0.505, rotation=<rot 0 -90 0>}]
 thigh_r [{shapeType=Capsule, length=30.6421, radius=8.6589, center=<16.9193 0.0309 -0.8121>, rotation=<rot 0 -88.2306 -83.0269>}]
}
constraints [
 {bone1Name=dyn_sword, bone2Name=dyn_sword_MATR, swing1LimitDegrees=45, swing2LimitDegrees=20, twistLimitDegrees=5}
 {
  bone1Name=dyn_sword_MATR
  bone2Name=thigh_r
  swing1LimitDegrees=45
  swing1Motion=Free
  swing2LimitDegrees=45
  swing2Motion=Free
  twistLimitDegrees=45
  twistMotion=Free
 }
 {
  bone1Name=calf_r
  bone2Name=thigh_r
  swing1LimitDegrees=45
  swing1Motion=Free
  swing2LimitDegrees=45
  swing2Motion=Free
  twistLimitDegrees=45
  twistMotion=Free
 }
]
