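# Post-copy AnimBP for the male medium Myth cosmetic body, adding secondary motion and deformation fixes to the parent mesh's pose.
# Constraints position sword, bag and band master bones, then AnimDynamics swings bag and bands; RigidBody nodes sit unconnected to the output.
# Head yaw and measured bone distances drive additive collar, tricep, lat and pectoral correctives.
# Weapon, customization and skydiving events toggle a material hide parameter based on the Myth pickaxe tag.
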AnimBlueprint /BRCosmetics/Characters/Player/Male/Medium/Bodies/M_MED_Myth/Meshes/M_MED_Myth_AnimBP
target_skeleton=&/M_MED_Myth_Skeleton
graphs [
 AnimGraph {nodes [
   "Constraint - dyn_band_r_MSTR (PPPPPP)"!(
    ComponentPose
    Alpha=1
    ConstraintWeights_0=0.3768
    ConstraintWeights_1:float=0.0002
    ConstraintWeights_2:float=0.2889
    ConstraintWeights_3:float=0.3154
    ConstraintWeights_4:float=0.0113
    ConstraintWeights_5:float=0.0074
   ) -> (Pose) {
    alphaBoolBlend {blendOption=Linear}
    boneToModify {boneName=dyn_band_r_MSTR}
    constraintSetup [
     {targetBone {boneName=upperarm_r}}
     {targetBone {boneName=lowerarm_r}}
     {targetBone {boneName=upperarm_twist_01_r}}
     {targetBone {boneName=upperarm_twist_02_r}}
     {targetBone {boneName=spine_05}}
     {targetBone {boneName=clavicle_r}}
    ]
    constraintWeights [1,1,1,1,1,1]
   }
   n1 = RigidBody(Alpha=Get FtndDynAlpha() -> (FtndDynAlpha:double)) {
    bClampLinearTranslationLimitToRefPose=true
    bOverrideWorldGravity=true
    overridePhysicsAsset=&/M_MED_Myth_dyn_bag
    simulationSpace=BaseBoneSpace
    alphaBoolBlend {blendOption=Linear}
    baseBoneRef {boneName=spine_01}
    componentAppliedLinearAccClamp=<3000 3000 3000>
    componentLinearAccScale=<0.3 0.3 0.3>
    componentLinearVelScale=<0.3 0.3 0.3>
    overrideWorldGravity {z=-980}
   }
   n2 = RigidBody(ComponentPose=@n1) {
    bClampLinearTranslationLimitToRefPose=true
    bOverrideWorldGravity=true
    overridePhysicsAsset=&/M_MED_Myth_dyn_band_r
    simulationSpace=BaseBoneSpace
    alphaBoolBlend {blendOption=Linear}
    baseBoneRef {boneName=upperarm_r}
    componentAppliedLinearAccClamp=<100 100 100>
    componentLinearAccScale=<0.3 0.3 0.3>
    componentLinearVelScale=<0.3 0.3 0.3>
    overrideWorldGravity {z=-100}
   }
   RigidBody(ComponentPose=@n2) {
    bClampLinearTranslationLimitToRefPose=true
    bOverrideWorldGravity=true
    overridePhysicsAsset=&/M_MED_Myth_dyn_band_l
    simulationSpace=BaseBoneSpace
    alphaBoolBlend {blendOption=Linear}
    baseBoneRef {boneName=upperarm_l}
    componentAppliedLinearAccClamp=<100 100 100>
    componentLinearAccScale=<0.3 0.3 0.3>
    componentLinearVelScale=<0.3 0.3 0.3>
    overrideWorldGravity {z=-100}
   }
   RigidBody() {
    bClampLinearTranslationLimitToRefPose=true
    bOverrideWorldGravity=true
    overridePhysicsAsset=&/M_MED_Myth_dyn_collar
    simulationSpace=BaseBoneSpace
    alphaBoolBlend {blendOption=Linear}
    baseBoneRef {boneName=spine_05}
    componentAppliedLinearAccClamp=<500 500 500>
    componentLinearAccScale=<0.1 0.1 0.1>
    componentLinearVelScale=<0.1 0.1 0.1>
    overrideWorldGravity {z=-980}
   }
   RigidBody() {
    bClampLinearTranslationLimitToRefPose=true
    bOverrideWorldGravity=true
    overridePhysicsAsset=&/M_MED_Myth_dyn_sword
    simulationSpace=BaseBoneSpace
    alphaBoolBlend {blendOption=Linear}
    baseBoneRef {boneName=thigh_r}
    componentAppliedLinearAccClamp=<3000 3000 3000>
    componentLinearAccScale=<0.5 0.5 0.3>
    componentLinearVelScale=<0.2 0.2 0.2>
    overrideWorldGravity {z=-980}
   }
   n3 = "Constraint - dyn_sword_MATR (PP)"!(
    ComponentPose=Local To Component(LocalPose=Copy Pose From Mesh() {bUseAttachedParent=true})
    Alpha=1
    ConstraintWeights_0=0.2061
    ConstraintWeights_1:float=0.7939
   ) -> (Pose) {
    alphaBoolBlend {blendOption=Linear}
    boneToModify {boneName=dyn_sword_MATR}
    constraintSetup [{targetBone {boneName=pelvis}}, {targetBone {boneName=thigh_twist_01_r}}]
    constraintWeights [1,1]
   }
   n4 = "Constraint - dyn_bag_MSTR (PPP)"!(ComponentPose=@n3, Alpha=1, ConstraintWeights_0=0.1678, ConstraintWeights_1:float=0.0263, ConstraintWeights_2:float=0.8059) -> (Pose) {
    alphaBoolBlend {blendOption=Linear}
    boneToModify {boneName=dyn_bag_MSTR}
    constraintSetup [{targetBone {boneName=spine_01}}, {targetBone {boneName=spine_02}}, {targetBone {boneName=pelvis}}]
    constraintWeights [1,1,1]
   }
   n5 = "Constraint - dyn_band_r_MSTR (PPPPPPP)"!(
    ComponentPose=@n4
    Alpha=1
    ConstraintWeights_0=0.2575
    ConstraintWeights_1:float=0.2777
    ConstraintWeights_2:float=0
    ConstraintWeights_3:float=0.1386
    ConstraintWeights_4:float=0.2987
    ConstraintWeights_5:float=0.0181
    ConstraintWeights_6:float=0.0094
   ) -> (Pose) {
    alphaBoolBlend {blendOption=Linear}
    boneToModify {boneName=dyn_band_r_MSTR}
    constraintSetup [
     {targetBone {boneName=tricep_01_r}}
     {targetBone {boneName=upperarm_r}}
     {targetBone {boneName=lowerarm_r}}
     {targetBone {boneName=upperarm_twist_01_r}}
     {targetBone {boneName=upperarm_twist_02_r}}
     {targetBone {boneName=spine_05}}
     {targetBone {boneName=clavicle_r}}
    ]
    constraintWeights [1,1,1,1,1,1,1]
   }
   n6 = "Constraint - dyn_band_l_MSTR (PPPPPP)"!(
    ComponentPose=@n5
    Alpha=1
    ConstraintWeights_0=0.0079
    ConstraintWeights_1:float=0.3131
    ConstraintWeights_2:float=0.0128
    ConstraintWeights_3:float=0.3092
    ConstraintWeights_4:float=0.0001
    ConstraintWeights_5:float=0.3569
   ) -> (Pose) {
    alphaBoolBlend {blendOption=Linear}
    boneToModify {boneName=dyn_band_l_MSTR}
    constraintSetup [
     {targetBone {boneName=clavicle_l}}
     {targetBone {boneName=upperarm_twist_02_l}}
     {targetBone {boneName=spine_05}}
     {targetBone {boneName=upperarm_twist_01_l}}
     {targetBone {boneName=lowerarm_l}}
     {targetBone {boneName=upperarm_l}}
    ]
    constraintWeights [1,1,1,1,1,1]
   }
   n7 = Anim Dynamics - Bone: dyn_bag_base!(ComponentPose=@n6, Alpha=1, GravityScale:float=1, GravityOverride:Vector="0,0,-980", ExternalForce="0,0,0") -> (Pose) {
    angularDampingOverride=0.9999
    angularSpringConstant=1000
    bAngularSpring=true
    bGravityOverrideInSimSpace=true
    bOverrideAngularDamping=true
    simulationSpace=RootRelative
    alphaBoolBlend {blendOption=Linear}
    boundBone {boneName=dyn_bag_base}
    componentAppliedLinearAccClamp=<5000 5000 5000>
    componentLinearAccScale=<0.8 0.2 0.8>
    componentLinearVelScale=<0.6 0.2 0.6>
    gravityOverride {z=-980}
    physicsBodyDefinitions [{
      sphereCollisionRadius=0
      boundBone {boneName=dyn_bag_base}
      boxExtents=<12 20 10>
      constraintSetup {angularLimitsMax=<0 0 0>, angularLimitsMin=<0 -10 0>, angularTarget=<0 0 0>}
      localJointOffset=<1.5 0 0>
     }]
    relativeSpaceBone {boneName=pelvis}
   }
   n8 = Anim Dynamics - Bone: dyn_bag_flap!(ComponentPose=@n7, Alpha=1, GravityScale:float=1, GravityOverride:Vector="0,0,-980", ExternalForce="0,0,0") -> (Pose) {
    angularDampingOverride=0.9999
    angularSpringConstant=1000
    bAngularSpring=true
    bGravityOverrideInSimSpace=true
    bOverrideAngularDamping=true
    simulationSpace=RootRelative
    alphaBoolBlend {blendOption=Linear}
    boundBone {boneName=dyn_bag_flap}
    componentAppliedLinearAccClamp=<5000 5000 5000>
    componentLinearAccScale=<0.8 0.2 0.8>
    componentLinearVelScale=<0.6 0.2 0.6>
    gravityOverride {z=-980}
    physicsBodyDefinitions [{
      sphereCollisionRadius=0
      boundBone {boneName=dyn_bag_flap}
      boxExtents=<12 20 10>
      constraintSetup {angularLimitsMax=<0 3 0>, angularLimitsMin=<0 -25 0>, angularTarget=<0 0 0>}
      localJointOffset=<1.6 0 -0.8>
     }]
    relativeSpaceBone {boneName=pelvis}
   }
   n9 = Anim Dynamics - Bone: dyn_band_l_a_01!(ComponentPose=@n8, Alpha=1, GravityScale:float=1, GravityOverride:Vector="0,0,-100", ExternalForce="0,0,0") -> (Pose) {
    angularDampingOverride=0.999
    angularSpringConstant=1500
    bAngularSpring=true
    bGravityOverrideInSimSpace=true
    bOverrideAngularDamping=true
    simulationSpace=RootRelative
    alphaBoolBlend {blendOption=Linear}
    boundBone {boneName=dyn_band_l_a_01}
    componentAppliedLinearAccClamp=<3000 3000 3000>
    componentLinearAccScale=<0.8 0.8 0.8>
    componentLinearVelScale=<0.5 0.5 0.5>
    gravityOverride {z=-980}
    physicsBodyDefinitions [{
      sphereCollisionRadius=0
      boundBone {boneName=dyn_band_l_a_01}
      boxExtents=<10 10 10>
      constraintSetup {angularLimitsMax=<5 35 35>, angularLimitsMin=<-5 -35 -35>, angularTarget=<0 0 0>}
      localJointOffset=<0.6 0.5 -0.2>
     }]
    relativeSpaceBone {boneName=upperarm_twist_02_l}
   }
   n10 = Anim Dynamics - Bone: dyn_band_l_b_01!(ComponentPose=@n9, Alpha=1, GravityScale:float=1, GravityOverride:Vector="0,0,-100", ExternalForce="0,0,0") -> (Pose) {
    angularDampingOverride=0.999
    angularSpringConstant=1500
    bAngularSpring=true
    bGravityOverrideInSimSpace=true
    bOverrideAngularDamping=true
    simulationSpace=RootRelative
    alphaBoolBlend {blendOption=Linear}
    boundBone {boneName=dyn_band_l_b_01}
    componentAppliedLinearAccClamp=<3000 3000 3000>
    componentLinearAccScale=<0.8 0.8 0.8>
    componentLinearVelScale=<0.5 0.5 0.5>
    gravityOverride {z=-980}
    physicsBodyDefinitions [{
      sphereCollisionRadius=0
      boundBone {boneName=dyn_band_l_b_01}
      boxExtents=<10 10 10>
      constraintSetup {angularLimitsMax=<5 15 35>, angularLimitsMin=<-5 -25 -35>, angularTarget=<0 0 0>}
      localJointOffset=<0 0.5 -0.2>
     }]
    relativeSpaceBone {boneName=upperarm_twist_02_l}
   }
   n11 = Anim Dynamics - Bone: dyn_band_r_a_01!(ComponentPose=@n10, Alpha=1, GravityScale:float=1, GravityOverride:Vector="0,0,-100", ExternalForce="0,0,0") -> (Pose) {
    angularDampingOverride=0.999
    angularSpringConstant=1500
    bAngularSpring=true
    bGravityOverrideInSimSpace=true
    bOverrideAngularDamping=true
    simulationSpace=RootRelative
    alphaBoolBlend {blendOption=Linear}
    boundBone {boneName=dyn_band_r_a_01}
    componentAppliedLinearAccClamp=<3000 3000 3000>
    componentLinearAccScale=<0.8 0.8 0.8>
    componentLinearVelScale=<0.5 0.5 0.5>
    gravityOverride {z=-980}
    physicsBodyDefinitions [{
      sphereCollisionRadius=0
      boundBone {boneName=dyn_band_r_a_01}
      boxExtents=<10 10 10>
      constraintSetup {angularLimitsMax=<5 25 35>, angularLimitsMin=<-5 -5 -35>, angularTarget=<0 0 0>}
      localJointOffset=<0.1 -0.3 -0.3>
     }]
    relativeSpaceBone {boneName=upperarm_twist_02_r}
   }
   n12 = Anim Dynamics - Bone: dyn_band_r_b_01!(ComponentPose=@n11, Alpha=1, GravityScale:float=1, GravityOverride:Vector="0,0,-100", ExternalForce="0,0,0") -> (Pose) {
    angularDampingOverride=0.999
    angularSpringConstant=1500
    bAngularSpring=true
    bGravityOverrideInSimSpace=true
    bOverrideAngularDamping=true
    simulationSpace=RootRelative
    alphaBoolBlend {blendOption=Linear}
    boundBone {boneName=dyn_band_r_b_01}
    componentAppliedLinearAccClamp=<3000 3000 3000>
    componentLinearAccScale=<0.8 0.8 0.8>
    componentLinearVelScale=<0.5 0.5 0.5>
    gravityOverride {z=-980}
    physicsBodyDefinitions [{
      sphereCollisionRadius=0
      boundBone {boneName=dyn_band_r_b_01}
      boxExtents=<10 10 10>
      constraintSetup {angularLimitsMax=<5 15 35>, angularLimitsMin=<-5 -25 -35>, angularTarget=<0 0 0>}
      localJointOffset=<0.3 -0.5 -0.2>
     }]
    relativeSpaceBone {boneName=upperarm_twist_02_r}
   }
   n13 = Get head_rot_up() -> (head_rot_up:double)
   n14 = "Transform (Modify) Bone - Bone: dyn_collar_bk_MATR"(ComponentPose=@n12, Alpha=@n13, Translation="0,0,0", Rotation="45,0,0", Scale="0.6,1,1") {
    rotationMode=BMM_Additive
    rotationSpace=BCS_BoneSpace
    scaleMode=BMM_Additive
    scaleSpace=BCS_BoneSpace
    alphaBoolBlend {blendOption=Linear}
    boneToModify {boneName=dyn_collar_bk_MATR}
    rotation {pitch=45}
    scale {x=0.6}
   }
   n15 = "Transform (Modify) Bone - Bone: dyn_collar_r_01_MATR"(ComponentPose=@n14, Alpha=@n13, Translation="0,0,0", Rotation="25,0,0", Scale="0.7,1,1") {
    rotationMode=BMM_Additive
    rotationSpace=BCS_BoneSpace
    scaleMode=BMM_Additive
    scaleSpace=BCS_BoneSpace
    alphaBoolBlend {blendOption=Linear}
    boneToModify {boneName=dyn_collar_r_01_MATR}
    rotation {pitch=25}
    scale {x=0.7}
   }
   n16 = "Transform (Modify) Bone - Bone: dyn_collar_l_01_MATR"(ComponentPose=@n15, Alpha=@n13, Translation="0,0,0", Rotation="0,0,0", Scale="0.7,1,1") {
    scaleMode=BMM_Additive
    scaleSpace=BCS_BoneSpace
    alphaBoolBlend {blendOption=Linear}
    boneToModify {boneName=dyn_collar_l_01_MATR}
    scale {x=0.7}
   }
   n17 = Get RlatStretchAlpha() -> (RlatStretchAlpha:double)
   n18 = "Transform (Modify) Bone - Bone: tricep_01_r"(ComponentPose=@n16, Alpha=@n17, Translation="0,-8,-4.5", Rotation="0,0,0", Scale="1,1,1") {
    translationMode=BMM_Additive
    translationSpace=BCS_BoneSpace
    alphaBoolBlend {blendOption=Linear}
    boneToModify {boneName=tricep_01_r}
    translation {y=-8, z=-4.5}
   }
   n19 = "Transform (Modify) Bone - Bone: lattissimus_01_r"(ComponentPose=@n18, Alpha=@n17, Translation="-10,0,5", Rotation="0,0,0", Scale="1,1,1") {
    translationMode=BMM_Additive
    translationSpace=BCS_BoneSpace
    alphaBoolBlend {blendOption=Linear}
    boneToModify {boneName=lattissimus_01_r}
    translation {x=-10, z=5}
   }
   n20 = "Transform (Modify) Bone - Bone: pectoral_01_r"(
    ComponentPose=@n19
    Alpha=Get PectStretchAlpha() -> (PectStretchAlpha:double)
    Translation="-19,-2.5,0"
    Rotation="-24.000,10.000,30"
    Scale="1,1,1"
   ) {
    rotationMode=BMM_Additive
    rotationSpace=BCS_BoneSpace
    translationMode=BMM_Additive
    translationSpace=BCS_BoneSpace
    alphaBoolBlend {blendOption=Linear}
    boneToModify {boneName=pectoral_01_r}
    rotation=<rot -24 10 30>
    translation=<-19 -2.5>
   }
   n21 = Get RlatStretchAlphaLeft() -> (RlatStretchAlphaLeft:double)
   n22 = "Transform (Modify) Bone - Bone: tricep_01_l"(ComponentPose=@n20, Alpha=@n21, Translation="0,8,4.5", Rotation="0,0,0", Scale="1,1,1") {
    translationMode=BMM_Additive
    translationSpace=BCS_BoneSpace
    alphaBoolBlend {blendOption=Linear}
    boneToModify {boneName=tricep_01_l}
    translation {y=8, z=4.5}
   }
   n23 = "Transform (Modify) Bone - Bone: lattissimus_01_l"(ComponentPose=@n22, Alpha=@n21, Translation="10,0,-5", Rotation="0,0,0", Scale="1,1,1") {
    translationMode=BMM_Additive
    translationSpace=BCS_BoneSpace
    alphaBoolBlend {blendOption=Linear}
    boneToModify {boneName=lattissimus_01_l}
    translation {x=10, z=-5}
   }
   n24 = "Transform (Modify) Bone - Bone: pectoral_01_l"(
    ComponentPose=@n23
    Alpha=Get PectStretchAlphaLeft() -> (PectStretchAlphaLeft:double)
    Translation="19,2.5,0"
    Rotation="-24,-10,-30"
    Scale="1,1,1"
   ) {
    rotationMode=BMM_Additive
    rotationSpace=BCS_BoneSpace
    translationMode=BMM_Additive
    translationSpace=BCS_BoneSpace
    alphaBoolBlend {blendOption=Linear}
    boneToModify {boneName=pectoral_01_l}
    rotation=<rot -24 -10 -30>
    translation=<19 2.5>
   }
   Output Pose(Result=Component To Local(ComponentPose=@n24)) {layerGroup=None}
  ]}
 EventGraph {nodes [
   Event BlueprintUpdateAnimation()
   n1 = BreakTransform(InTransform=GetDeltaTransformFromRefPose(self=GetOwnerSkeletalMeshComponent(), BoneName=head, BaseName=spine_05))
   > Set head_rot_up(head_rot_up:double=MapRangeClamped(Value=BreakRotator(InRot=@n1.Rotation).Yaw, InRangeA=0, InRangeB=-50)) -> (Output_Get:double)
   n2 = GetPartSkeletalMeshComponent()
   > n3 = TransformFromBoneSpace!(
    execute
    self:Skinned Mesh Component Object Reference=@n2
    BoneName:Name=dyn_tricep_measure_r
    InPosition:Vector="0, 0, 0"
    InRotation:Rotator="0, 0, 0"
   ) -> (then, OutPosition:Vector, OutRotation:Rotator)
   > n4 = TransformFromBoneSpace(self=@n2, BoneName=dyn_lat_measure_r, InPosition="0, 0, 0", InRotation="0, 0, 0")
   > Set RlatStretchAlpha(RlatStretchAlpha:double=MapRangeClamped(Value="Distance (Vector)"!(V1:Vector=@n3, V2:Vector=@n4) -> (ReturnValue:double), InRangeA=18, InRangeB=12)) -> (Output_Get:double)
   n5 = GetPartSkeletalMeshComponent()
   > n6 = TransformFromBoneSpace(self=@n5, BoneName=dyn_tricep_measure_r, InPosition="0, 0, 0", InRotation="0, 0, 0")
   > n7 = TransformFromBoneSpace(self=@n5, BoneName=dyn_pec_measure_r, InPosition="0, 0, 0", InRotation="0, 0, 0")
   > Set PectStretchAlpha(PectStretchAlpha:double=MapRangeClamped(Value="Distance (Vector)"(V1=@n6, V2=@n7), InRangeA=20, InRangeB=9)) -> (Output_Get:double)
   n8 = GetPartSkeletalMeshComponent()
   > n9 = TransformFromBoneSpace(self=@n8, BoneName=dyn_tricep_measure_l, InPosition="0, 0, 0", InRotation="0, 0, 0")
   > n10 = TransformFromBoneSpace(self=@n8, BoneName=dyn_lat_measure_l, InPosition="0, 0, 0", InRotation="0, 0, 0")
   > Set RlatStretchAlphaLeft(RlatStretchAlphaLeft:double=MapRangeClamped(Value="Distance (Vector)"(V1=@n9, V2=@n10), InRangeA=18, InRangeB=12)) -> (Output_Get:double)
   n11 = GetPartSkeletalMeshComponent()
   > n12 = TransformFromBoneSpace(self=@n11, BoneName=dyn_tricep_measure_l, InPosition="0, 0, 0", InRotation="0, 0, 0")
   > n13 = TransformFromBoneSpace(self=@n11, BoneName=dyn_pec_measure_l, InPosition="0, 0, 0", InRotation="0, 0, 0")
   > Set PectStretchAlphaLeft(PectStretchAlphaLeft:double=MapRangeClamped(Value="Distance (Vector)"(V1=@n12, V2=@n13), InRangeA=20, InRangeB=9)) -> (Output_Get:double)
   > n14 = Branch(Condition=Get bIsInFrontEnd())
   > Set FtndDynAlpha(FtndDynAlpha:double=0.25) -> (Output_Get:double)
   Set FtndDynAlpha(execute=@n14.else, FtndDynAlpha=1)
   n15 = Event OnWeaponChanged()
   n16 = Get Dynamic Material Instance(execute=[@n15.then, @n24.then], SkeletalMeshComponent=GetPartSkeletalMeshComponent(), ElementIndex=1)
   > n17 = Sequence()
   n18 = PickaxeItemDefinitionHasAnyMetaTag(Tags=Make Array("[0]"=(TagName="Cosmetics.MetaTags.MythMale")), bRequiresEquipped=true)
   > n19 = Branch(Condition=@n18)
   > n20 = Branch(Condition=Get bIsSkyDiving())
   SetScalarParameterValue(execute=[@n22.then, @n20.then, @n21.else, @n19.else], self=@n16, ParameterName=Hide Element 01, Value=0)
   SetScalarParameterValue(execute=[@n20.else, @n21.then], self=@n16, ParameterName=Hide Element 01, Value=1)
   n21 = Branch(execute=[@n23.then, @n17.then_1], Condition=@n18)
   n22 = Event OnBeginSkydiving()
   n23 = Event OnEndSkydiving()
   n24 = Event OnCharacterCustomizationLoaded()
  ]}
]
variables [
 FtndDynAlpha {type=number}
 head_rot_up {type=number}
 PectStretchAlpha {type=number}
 PectStretchAlphaLeft {type=number}
 RlatStretchAlpha {type=number}
 RlatStretchAlphaLeft {type=number}
]
