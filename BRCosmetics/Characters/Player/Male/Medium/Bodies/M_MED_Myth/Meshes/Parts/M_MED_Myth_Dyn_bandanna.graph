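# Physics asset for the dynamic bandanna and collar on the Myth medium male player body.
# Two five-capsule bandanna chains hang from a master body linked to the head, with wider swing toward the tips.
# Five collar flaps each attach to spine_05 through their own anchor body, with narrow twist and one wide swing axis for hinge-like motion.
# Head, neck, clavicle and upper spine capsules supply the body shapes around them.
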PhysicsAsset /BRCosmetics/Characters/Player/Male/Medium/Bodies/M_MED_Myth/Meshes/Parts/M_MED_Myth_Dyn_bandanna
bodies {
 clavicle_l [{shapeType=Capsule, length=6.5, radius=9.25, center=<8.39 -1.7757 1.047>, rotation=<rot 79.9995 0.0001 0>}]
 clavicle_r [{shapeType=Capsule, length=6.5, radius=9.25, center=<-8.163 2.2376 0.2404>, rotation=<rot 79.9995 0.0001 0>}]
 dyn_bandanna_MSTR [{shapeType=Capsule, length=0.5, radius=0.5}]
 dyn_bandanna_l_01 [{shapeType=Capsule, length=1, radius=1, shapeName=dyn_bandanna_l_01_capsule, center=<2 0 0>, rotation=<rot 90 0 0>}]
 dyn_bandanna_l_02 [{shapeType=Capsule, length=2, radius=1, shapeName=dyn_bandanna_l_02_capsule, center=<2 0 0>, rotation=<rot 90 0 0>}]
 dyn_bandanna_l_03 [{shapeType=Capsule, length=2.5, radius=1.5, shapeName=dyn_bandanna_l_03_capsule, center=<2 0 0.7>, rotation=<rot 85.0001 -179.9999 -180>}]
 dyn_bandanna_l_04 [{shapeType=Capsule, length=2.875, radius=1.625, shapeName=dyn_bandanna_l_04_capsule, center=<2 0 0.7>, rotation=<rot 90 0 0>}]
 dyn_bandanna_l_05 [{shapeType=Capsule, length=4.125, radius=1.875, shapeName=dyn_bandanna_l_05_capsule, center=<2 0 0.7>, rotation=<rot 90 0 0>}]
 dyn_bandanna_r_01 [{shapeType=Capsule, length=1, radius=1, shapeName=dyn_bandanna_r_01_capsule, center=<2 0 0>, rotation=<rot 90 0 0>}]
 dyn_bandanna_r_02 [{shapeType=Capsule, length=2, radius=1, shapeName=dyn_bandanna_r_02_capsule, center=<2 0 0>, rotation=<rot 90 0 0>}]
 dyn_bandanna_r_03 [{shapeType=Capsule, length=2.5, radius=1.5, shapeName=dyn_bandanna_r_03_capsule, center=<2 0 0.7>, rotation=<rot 85 180 180>}]
 dyn_bandanna_r_04 [{shapeType=Capsule, length=2.875, radius=1.625, shapeName=dyn_bandanna_r_04_capsule, center=<2 0 0.7>, rotation=<rot 90 0 0>}]
 dyn_bandanna_r_05 [{shapeType=Capsule, length=4.125, radius=1.875, shapeName=dyn_bandanna_r_05_capsule, center=<2 0 0.7>, rotation=<rot 90 0 0>}]
 dyn_collar_bk [{shapeType=Capsule, length=12.125, radius=3.375, center=<5.4185 0 -1.5337>, rotation=<rot 0 0 90>}]
 dyn_collar_bk_MATR [{shapeType=Capsule, length=0.5, radius=0.5, rotation=<rot 0 -90 0>}]
 dyn_collar_l_01 [{shapeType=Capsule, length=6.75, radius=2.75, center=<5.3 0 0>, rotation=<rot 0 -10 85>}]
 dyn_collar_l_01_MATR [{shapeType=Capsule, length=0.5, radius=0.5, rotation=<rot 0 -90 0>}]
 dyn_collar_l_02 [{shapeType=Capsule, length=7, radius=2.75, center=<3.5 0 0>, rotation=<rot -1.2926 -4.8305 75.054>}]
 dyn_collar_l_02_MATR [{shapeType=Capsule, length=0.5, radius=0.5, rotation=<rot 0 -90 0>}]
 dyn_collar_r_01 [{shapeType=Capsule, length=6.375, radius=3.125, center=<4.9277 0.0085 0.0129>, rotation=<rot -1.7277 0.1487 80.149>}]
 dyn_collar_r_01_MATR [{shapeType=Capsule, length=0.5, radius=0.5, rotation=<rot 0 -90 0>}]
 dyn_collar_r_02 [{shapeType=Capsule, length=7, radius=2.75, center=<3.5 0 0>, rotation=<rot -1.2925 4.8308 104.9462>}]
 dyn_collar_r_02_MATR [{shapeType=Capsule, length=0.5, radius=0.5, rotation=<rot 0 -90 0>}]
 head [
  {shapeType=Capsule, length=4.25, radius=7, center=<5.8336 1.593 0.0002>, rotation=<rot 90 0 20.0066>}
  {shapeType=Capsule, length=7, radius=5.5, center=<3.5161 7.1681 0.0001>, rotation=<rot 90 0 0>}
 ]
 neck_01 [{shapeType=Capsule, length=6.125, radius=10.125, center=<5.6532 1.4691 0.0001>}]
 spine_04 [{shapeType=Capsule, length=17.4328, radius=11.8515, center=<7.2234 -1.2716 0>}]
 spine_05 [{shapeType=Capsule, length=13.6828, radius=11.8515, center=<5.5789 1.2672 0>}]
}
constraints [
 {bone1Name=head, bone2Name=neck_01, swing1LimitDegrees=45, swing2LimitDegrees=45, twistLimitDegrees=45}
 {
  bone1Name=dyn_bandanna_MSTR
  bone2Name=head
  swing1LimitDegrees=45
  swing1Motion=Free
  swing2LimitDegrees=45
  swing2Motion=Free
  twistLimitDegrees=45
  twistMotion=Free
 }
 {bone1Name=dyn_bandanna_l_01, bone2Name=dyn_bandanna_MSTR, swing1LimitDegrees=5, swing2LimitDegrees=45, twistLimitDegrees=5}
 {bone1Name=dyn_bandanna_l_02, bone2Name=dyn_bandanna_l_01, swing1LimitDegrees=5, swing2LimitDegrees=45, twistLimitDegrees=5}
 {bone1Name=dyn_bandanna_l_03, bone2Name=dyn_bandanna_l_02, swing1LimitDegrees=8, swing2LimitDegrees=60, twistLimitDegrees=5}
 {bone1Name=dyn_bandanna_l_04, bone2Name=dyn_bandanna_l_03, swing1LimitDegrees=8, swing2LimitDegrees=90, twistLimitDegrees=5}
 {bone1Name=dyn_bandanna_r_01, bone2Name=dyn_bandanna_MSTR, swing1LimitDegrees=5, swing2LimitDegrees=45, twistLimitDegrees=5}
 {bone1Name=dyn_bandanna_r_02, bone2Name=dyn_bandanna_r_01, swing1LimitDegrees=5, swing2LimitDegrees=45, twistLimitDegrees=5}
 {bone1Name=dyn_bandanna_r_03, bone2Name=dyn_bandanna_r_02, swing1LimitDegrees=8, swing2LimitDegrees=60, twistLimitDegrees=5}
 {bone1Name=dyn_bandanna_r_04, bone2Name=dyn_bandanna_r_03, swing1LimitDegrees=8, swing2LimitDegrees=90, twistLimitDegrees=5}
 {
  bone1Name=dyn_collar_bk_MATR
  bone2Name=spine_05
  swing1LimitDegrees=45
  swing1Motion=Free
  swing2LimitDegrees=45
  swing2Motion=Free
  twistLimitDegrees=45
  twistMotion=Free
 }
 {bone1Name=dyn_collar_bk, bone2Name=dyn_collar_bk_MATR, swing1LimitDegrees=2, swing2LimitDegrees=90, twistLimitDegrees=2}
 {
  bone1Name=dyn_collar_r_01_MATR
  bone2Name=spine_05
  swing1LimitDegrees=45
  swing1Motion=Free
  swing2LimitDegrees=45
  swing2Motion=Free
  twistLimitDegrees=45
  twistMotion=Free
 }
 {bone1Name=dyn_collar_r_01, bone2Name=dyn_collar_r_01_MATR, swing1LimitDegrees=2, swing2LimitDegrees=90, twistLimitDegrees=2}
 {
  bone1Name=dyn_collar_l_01_MATR
  bone2Name=spine_05
  swing1LimitDegrees=45
  swing1Motion=Free
  swing2LimitDegrees=45
  swing2Motion=Free
  twistLimitDegrees=45
  twistMotion=Free
 }
 {bone1Name=dyn_collar_l_01, bone2Name=dyn_collar_l_01_MATR, swing1LimitDegrees=2, swing2LimitDegrees=90, twistLimitDegrees=2}
 {
  bone1Name=clavicle_l
  bone2Name=spine_05
  swing1LimitDegrees=45
  swing1Motion=Free
  swing2LimitDegrees=45
  swing2Motion=Free
  twistLimitDegrees=45
  twistMotion=Free
 }
 {
  bone1Name=clavicle_r
  bone2Name=spine_05
  swing1LimitDegrees=45
  swing1Motion=Free
  swing2LimitDegrees=45
  swing2Motion=Free
  twistLimitDegrees=45
  twistMotion=Free
 }
 {
  bone1Name=dyn_collar_r_02_MATR
  bone2Name=spine_05
  swing1LimitDegrees=45
  swing1Motion=Free
  swing2LimitDegrees=45
  swing2Motion=Free
  twistLimitDegrees=45
  twistMotion=Free
 }
 {bone1Name=dyn_collar_r_02, bone2Name=dyn_collar_r_02_MATR, swing1LimitDegrees=2, swing2LimitDegrees=90, twistLimitDegrees=2}
 {
  bone1Name=dyn_collar_l_02_MATR
  bone2Name=spine_05
  swing1LimitDegrees=45
  swing1Motion=Free
  swing2LimitDegrees=45
  swing2Motion=Free
  twistLimitDegrees=45
  twistMotion=Free
 }
 {bone1Name=dyn_collar_l_02, bone2Name=dyn_collar_l_02_MATR, swing1LimitDegrees=2, swing2LimitDegrees=90, twistLimitDegrees=2}
 {bone1Name=dyn_bandanna_r_05, bone2Name=dyn_bandanna_r_04, swing1LimitDegrees=8, swing2LimitDegrees=90, twistLimitDegrees=5}
 {bone1Name=dyn_bandanna_l_05, bone2Name=dyn_bandanna_l_04, swing1LimitDegrees=8, swing2LimitDegrees=90, twistLimitDegrees=5}
 {
  bone1Name=neck_01
  bone2Name=spine_05
  swing1LimitDegrees=45
  swing1Motion=Free
  swing2LimitDegrees=45
  swing2Motion=Free
  twistLimitDegrees=45
  twistMotion=Free
 }
 {
  bone1Name=spine_05
  bone2Name=spine_04
  swing1LimitDegrees=45
  swing1Motion=Free
  swing2LimitDegrees=45
  swing2Motion=Free
  twistLimitDegrees=45
  twistMotion=Free
 }
]
