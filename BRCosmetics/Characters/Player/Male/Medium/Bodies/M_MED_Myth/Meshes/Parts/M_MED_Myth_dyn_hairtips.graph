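# Physics asset for the hair tips part of the M_MED_Myth male player cosmetic body, defining how its dynamic hair bones can swing.
# Each of the ten dyn_hair_tips bones has a small capsule body, and the head bone has a capsule as the shared anchor.
# Every tip is constrained directly to the head rather than chained, using identical swing and twist limits so each moves independently within a limited range.
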PhysicsAsset /BRCosmetics/Characters/Player/Male/Medium/Bodies/M_MED_Myth/Meshes/Parts/M_MED_Myth_dyn_hairtips
bodies {
 dyn_hair_tips_01 [{shapeType=Capsule, length=1.5, radius=1.25, center=<1.6346 0 0>, rotation=<rot 90 0 0>}]
 dyn_hair_tips_010 [{shapeType=Capsule, length=1, radius=1, center=<0.8636 -0.0001 0>, rotation=<rot 90 0 0>}]
 dyn_hair_tips_02 [{shapeType=Capsule, length=1, radius=1, center=<1.1675 0 0>, rotation=<rot 90 0 0>}]
 dyn_hair_tips_03 [{shapeType=Capsule, length=1.75, radius=1.5, center=<2.1372 -0.0001 0>, rotation=<rot 90 0.0049 0.0049>}]
 dyn_hair_tips_04 [{shapeType=Capsule, length=0.75, radius=1.25, center=<1.5044 0 0>, rotation=<rot 90 0 0>}]
 dyn_hair_tips_05 [{shapeType=Capsule, length=1, radius=1, center=<1.5657 0 -0.0001>, rotation=<rot 90 0 0>}]
 dyn_hair_tips_06 [{shapeType=Capsule, length=1.75, radius=1.25, center=<1.9653 0 0>, rotation=<rot 90 0 0>}]
 dyn_hair_tips_07 [{shapeType=Capsule, length=1, radius=1, center=<0.7673 0 0>, rotation=<rot 90 0 0>}]
 dyn_hair_tips_08 [{shapeType=Capsule, length=1, radius=1, center=<1.0257 0 0>, rotation=<rot 90 0 0>}]
 dyn_hair_tips_09 [{shapeType=Capsule, length=1.5, radius=1, center=<1.4721 0 0>, rotation=<rot 90 0 0>}]
 head [{shapeType=Capsule, length=1, radius=1}]
}
constraints [
 {bone1Name=dyn_hair_tips_01, bone2Name=head, swing1LimitDegrees=30.035, swing2LimitDegrees=35, twistLimitDegrees=20}
 {bone1Name=dyn_hair_tips_02, bone2Name=head, swing1LimitDegrees=30.035, swing2LimitDegrees=35, twistLimitDegrees=20}
 {bone1Name=dyn_hair_tips_03, bone2Name=head, swing1LimitDegrees=30.035, swing2LimitDegrees=35, twistLimitDegrees=20}
 {bone1Name=dyn_hair_tips_04, bone2Name=head, swing1LimitDegrees=30.035, swing2LimitDegrees=35, twistLimitDegrees=20}
 {bone1Name=dyn_hair_tips_05, bone2Name=head, swing1LimitDegrees=30.035, swing2LimitDegrees=35, twistLimitDegrees=20}
 {bone1Name=dyn_hair_tips_06, bone2Name=head, swing1LimitDegrees=30.035, swing2LimitDegrees=35, twistLimitDegrees=20}
 {bone1Name=dyn_hair_tips_07, bone2Name=head, swing1LimitDegrees=30.035, swing2LimitDegrees=35, twistLimitDegrees=20}
 {bone1Name=dyn_hair_tips_08, bone2Name=head, swing1LimitDegrees=30.035, swing2LimitDegrees=35, twistLimitDegrees=20}
 {bone1Name=dyn_hair_tips_09, bone2Name=head, swing1LimitDegrees=30.035, swing2LimitDegrees=35, twistLimitDegrees=20}
 {bone1Name=dyn_hair_tips_010, bone2Name=head, swing1LimitDegrees=30.035, swing2LimitDegrees=35, twistLimitDegrees=20}
]
